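# Player animation Blueprint for the tank's turret passenger, overriding the vehicle final-pose layer.
# Each update snaps the pawn to the tank's Passenger socket and reads speed and turret rotation from the vehicle and its AnimBP.
# These drive an idle blendspace, additive yaw lean and pitch aim offset, faded by the BothHandFK curve.
# Leg IK pulls hands to IK bones before inertialization; the socket-based hand-target nodes are disconnected.
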AnimBlueprint /Tank/Animations/Player/Tank_TurretPassenger_AnimBP
target_skeleton=&Fortnite_M_Avg_Player_Skeleton
graphs [
 AnimGraph {nodes [Output Pose() {layerGroup=None}]}
 EventGraph {nodes [
   Event BlueprintUpdateAnimation()
   n1 = Get FortPlayerPawn() -> (FortPlayerPawn:Fort Player Pawn Object Reference)
   n2 = GetVehicleActor!(self:Fort Player Pawn Object Reference=@n1) -> (ReturnValue:Actor Object Reference)
   > Is Valid(InputObject=@n2)
   > n3 = Set TankActor(TankActor:Actor Object Reference=@n2) -> (Output_Get:Actor Object Reference)
   > FindSeatIndex!(execute:Exec, self:Object Reference=@n3, PlayerPawn:Fort Player Pawn Object Reference=@n1) -> (then:Exec, ReturnValue:int=0) {is=title K2Node_Message}
   > n4 = Cast To Tank_Vehicle!(execute, Object=Get TankActor() -> (TankActor:Actor Object Reference)) -> (then, CastFailed, AsTank Vehicle:Tank Vehicle Object Reference) {is=title K2Node_DynamicCast}
   > n5 = Sequence()
   > Set Speed(Speed:double=GetForwardSpeedKmh!(self:Fort Vehicle Interface Interface=@n4) -> (ReturnValue:float)) -> (Output_Get:double)
   n6 = Get Montage_GunnerRightHandFire!(self:Tank Vehicle Object Reference=@n4) -> (Montage_GunnerRightHandFire:Anim Montage Object Reference)
   n7 = Get Montage_GunnerLeftHandFire!(self:Tank Vehicle Object Reference=@n4) -> (Montage_GunnerLeftHandFire:Anim Montage Object Reference)
   > Set IsShootingTurret(IsShootingTurret:bool=OR Boolean(
     A=Montage_IsActive!(self:Anim Instance Object Reference, Montage:Anim Montage Object Reference=@n7) -> (ReturnValue:bool)
     B=Montage_IsActive(Montage=@n6)
    )) -> (Output_Get:bool)
   n8 = GetComponentByClass!(self:Actor Object Reference=Get TankActor(), ComponentClass:Actor Component Class Reference=/Script/Engine.SkeletalMeshComponent) -> (ReturnValue:Skeletal Mesh Component Object Reference)
   n9 = Cast To Vehicle_Tank_AnimBP!(execute=@n5.then_1, Object=GetAnimInstance(self=@n8)) -> (then, CastFailed, AsVehicle Tank Anim BP:Vehicle Tank Anim BP Object Reference) {is=title K2Node_DynamicCast}
   > Set YawDelta(YawDelta:double=BreakRotator(InRot=Get SmoothedTurretDeltaRotation!(self:Fort Tank Vehicle Anim Instance Object Reference=@n9) -> (SmoothedTurretDeltaRotation:Rotator)).Yaw) -> (Output_Get:double)
   > Set TurretPitch(TurretPitch:double=BreakRotator(InRot=Get CurrentTurretAimRotation!(self:Fort Tank Vehicle Anim Instance Object Reference=@n9) -> (CurrentTurretAimRotation:Rotator)).Pitch) -> (Output_Get:double)
   > Set Actor Transform!(
    execute
    self:Actor Object Reference=Get FortPlayerPawn()
    NewTransform:"Transform (by ref)"=GetSocketTransform(self=@n8, InSocketName=Passenger, TransformSpace=RTS_World)
    bSweep:bool=false
    bTeleport:bool=false
   ) -> (then, SweepHitResult:Hit Result Structure, ReturnValue:bool) {is=title K2Node_CallFunction}
   n10 = Get SkeletalMesh!(self:Fort Athena Vehicle Object Reference) -> (SkeletalMesh:Skeletal Mesh Component Budgeted Object Reference)
   Set HandAttachL(HandAttachL:Vector=TransformLocation!(
     T:"Transform (by ref)"=GetSocketTransform(self=@n10, InSocketName=turret_gun_hand_attach_l, TransformSpace=RTS_World)
     Location:Vector="0, 0, 0"
    ) -> (ReturnValue:Vector)) -> (Output_Get:Vector)
   > Set HandAttachR(HandAttachR:Vector=TransformLocation(T=GetSocketTransform(self=@n10, InSocketName=turret_gun_hand_attach_r, TransformSpace=RTS_World), Location="0, 0, 0")) -> (Output_Get:Vector)
   Reroute Node(InputPin:Wildcard) -> (OutputPin:Wildcard)
  ]}
 Update Yaw Delta Smoothed {nodes [
   n1 = Update Yaw Delta Smoothed!() -> (
    then
    Vehicle:Fort Athena Vehicle Object Reference
    SocketName:Name
    NewRotation:Rotator=~
    SmoothedYawValue:"Float (double-precision) (by ref)"
   ) {is=title K2Node_FunctionEntry}
   n2 = "Delta (Rotator)"!(
    A:Rotator=GetSocketRotation!(self:Scene Component Object Reference=Get SkeletalMesh(self=@n1), InSocketName:Name=@n1.SocketName) -> (ReturnValue:Rotator)
    B:Rotator=@n1.NewRotation
   ) -> (ReturnValue:Rotator)
   > "Set Float (double-precision) (by ref)"!(
    execute:Exec
    Target:"Float (double-precision) (by ref)"=@n1.SmoothedYawValue
    Value:double=Lerp(A=BreakRotator(InRot=@n2).Yaw, B=@n1.SmoothedYawValue, Alpha=0.71)
   ) -> (then:Exec) {is=title K2Node_VariableSetRef}
   > Return Node!(execute, SmoothedYaw:double=0)
  ]}
 VehicleFinalPoseOverride {nodes [
   Input Pose() {name=InFinalPose}
   n1 = "Transform (Modify) Bone - Bone: ik_hand_r"!(
    ComponentPose
    Alpha=Get RightHandIKAlpha() -> (RightHandIKAlpha:float)
    Translation=Get HandAttachR() -> (HandAttachR:Vector)
    Rotation="0,0,0"
   ) -> (Pose) {translationMode=BMM_Replace, translationSpace=BCS_WorldSpace, alphaBoolBlend {blendOption=Linear}, boneToModify {boneName=ik_hand_r}}
   "Transform (Modify) Bone - Bone: ik_hand_l"!(
    ComponentPose=@n1
    Alpha=Get LeftHandIKAlpha() -> (LeftHandIKAlpha:float)
    Translation=Get HandAttachL() -> (HandAttachL:Vector)
    Rotation="0,0,0"
   ) -> (Pose) {translationMode=BMM_Replace, translationSpace=BCS_WorldSpace, alphaBoolBlend {blendOption=Linear}, boneToModify {boneName=ik_hand_l}}
   n2 = Get Speed() -> (Speed:double)
   n3 = Apply Additive(
    Base=Slot 'FullBody'(Source=Blendspace Player 'TankVehicle_TurretPassenger_Idle_BS'!(X=@n2) -> (Pose) {blendSpace=&/TankVehicle_TurretPassenger_Idle_BS}) {slotName=FullBody}
    Additive=Blendspace Player 'Tank_TurretPassenger_Lean_BS'!(X=Get YawDelta() -> (YawDelta:double)) -> (Pose) {blendSpace=/Tank/Animations/Player/AO/Tank_TurretPassenger_Lean_BS}
   ) {alphaBoolBlend {blendOption=Linear}}
   n4 = float + float(A=float * float(A=GetCurveValue(CurveName=BothHandFK), B=-1) {is=title K2Node_PromotableOperator}, B=1)
   n5 = Get TurretPitch() -> (TurretPitch:double)
   n6 = Leg IK!(
    ComponentPose:Component Space Pose Link Structure=Local To Component(LocalPose=AimOffset Player 'Tank_TurretPassenger_AO'!(X:float=@n5, BasePose:Pose Link Structure=@n3, Alpha:float=@n4) -> (Pose:Pose Link Structure) {blendSpace=/Tank/Animations/Player/AO/Tank_TurretPassenger_AO, alphaBoolBlend {blendOption=Linear}})
    Alpha:float=Get RightHandIKAlpha()
   ) -> (Pose:Component Space Pose Link Structure) {
    is=title AnimGraphNode_LegIK
    alphaBoolBlend {blendOption=Linear}
    legsDefinition [{
      bEnableKneeTwistCorrection=false
      bEnableRotationLimit=false
      footBoneForwardAxis=Y
      hingeRotationAxis=Z
      minRotationAngle=15
      numBonesInLimb=2
      twistOffsetCurveName=None
      fKFootBone {boneName=hand_r}
      iKFootBone {boneName=ik_hand_r}
     }]
   }
   n7 = Leg IK(ComponentPose=@n6, Alpha=Get LeftHandIKAlpha()) {
    alphaBoolBlend {blendOption=Linear}
    legsDefinition [{
      bEnableKneeTwistCorrection=false
      bEnableRotationLimit=false
      footBoneForwardAxis=Y
      hingeRotationAxis=Z
      minRotationAngle=15
      numBonesInLimb=2
      twistOffsetCurveName=None
      fKFootBone {boneName=hand_l}
      iKFootBone {boneName=ik_hand_l}
     }]
   }
   Output Pose(Result=Inertialization!(Source:Pose Link Structure=Component To Local(ComponentPose=@n7)) -> (Pose:Pose Link Structure) {is=title AnimGraphNode_Inertialization}) {layerGroup=VehicleOverride}
  ]}
 VehicleFullBodyOverride {nodes [
   Input Pose() {name=InPoseFullBody}
   Output Pose() {layerGroup=VehicleOverride}
  ]}
 VehicleLowerBodyOverride {nodes [
   Input Pose() {name=InPoseLowerBodyDefault}
   Input Pose() {name=InPoseUpperBody}
   Output Pose() {layerGroup=VehicleOverride}
  ]}
 VehicleSplitBodyOverride {nodes [
   Input Pose() {name=InPoseSplitBody}
   Input Pose() {name=InPoseUpperAndLowerBody}
   Input Pose() {name=InPoseUpperAndLowerBodyRemoveRoot}
   Output Pose() {layerGroup=Vehicle Override}
  ]}
]
variables [
 BattleBus {type=Actor, default=null}
 HandAttachL {type=Vector}
 HandAttachR {type=Vector}
 IsShootingTurret {type=boolean}
 PawnSeat {type=integer}
 ShouldCorrectRootYaw {type=boolean}
 SlopePitchDegreeAngle {type=number}
 SlopeRollDegreeAngle {type=number}
 SmoothedTurretPitch {type=number}
 Speed {type=number}
 TankActor {type=Actor, default=null}
 TankTurretPitch {type=number}
 TempAttachAlpha {type=number, default=1}
 TurretPassengerBack {type=integer, default=2}
 TurretPassengerFront {type=integer, default=1}
 TurretPitch {type=number}
 TurretYaw {type=number}
 YawDelta {type=number}
]
